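; Gameplay ability for a Season 26 quest that detects when the owning player eliminates someone who damaged them.
; Listens on the owning pawn for received, overshielded and shielded damage events, storing unique instigator names in AttackersList when the instigator tags include Homebase.Class.
; On each earned elimination, broadcasts CustomVerb.KilledPlayerAfterDamaged to the owner's controller if the target's name is listed.
; Matching is by player name, and the list is never cleared here.
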(event Ability|EventActivateAbility
  (bind _owningplayerpawn (Variables|GAVars|Data|GetOwningPlayerPawn))
  (|CallParentFunction)
  (bind (_asynctaskproxy _payload) (Ability|Async|WaitGameplayEventToActor _owningplayerpawn "(TagName=\"Event.Damage.Received\")")
    (:then
      (bind (_asynctaskproxy_2 _payload_2) (Ability|Async|WaitGameplayEventToActor _owningplayerpawn "(TagName=\"Event.Damage.Overshielded\")")
        (:then
          (bind (_asynctaskproxy_3 _payload_3) (Ability|Async|WaitGameplayEventToActor _owningplayerpawn "(TagName=\"Event.Damage.Shielded\")")
            (:then
              (bind (_asynctaskproxy_4 _payload_4) (Ability|Async|WaitGameplayEventToActor _owningplayerpawn "(TagName=\"Event.EarnedElimination\")" false false)
                (:EventReceived
                  (if (Utilities|Array|ContainsItem (Variables|Default|GetAttackersList) (Fortnite|GetPlayerNameSafe (.Target (Utilities|Struct|BreakGameplayEventData _payload_4))))
                    (CustomVerb|BroadcastCustomVerbMessage (CustomVerb|MakeCustomVerbMessage (Fortnite|GetFortPlayerControllerfromActor (Variables|GAVars|Data|GetOwningPlayerPawn)) "(TagName=\"CustomVerb.KilledPlayerAfterDamaged\")"))))))
            (:EventReceived
              (CallFunction|AddtoList _payload_3))))
        (:EventReceived
          (CallFunction|AddtoList _payload_2))))
    (:EventReceived
      (CallFunction|AddtoList _payload))))

(event Ability|EventOnEndAbility
  (params (bWasCancelled :type Boolean)))

(fn AddToList
  (params (NewParam :type Gameplay%20Event%20Data%20Structure))
  (bind (_eventtag _instigator _target _optionalobject _optionalobject2 _contexthandle _instigatortags) (Utilities|Struct|BreakGameplayEventData NewParam))
  (if (GameplayTags|HasTag _instigatortags "(TagName=\"Homebase.Class\")")
    (Utilities|Array|AddUnique (Variables|Default|GetAttackersList) (Fortnite|GetPlayerNameSafe _instigator))))
